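; Gameplay cue notify for opening a stamina container, going by its name, with only placeholder behavior of its own.
; OnActive emits a debug Print String, then returns the parent class's result.
; OnExecute only forwards the target and cue parameters to the parent, so any actual effects come from outside these two graphs.
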(fn OnActive
  (params (MyTarget :type Actor%20Object%20Reference) (Parameters :type Gameplay%20Cue%20Parameters%20Structure%20%28by%20ref%29))
  (Development|PrintString 0)
  (return (|CallParentFunction MyTarget Parameters)))

(fn OnExecute
  (params (MyTarget :type Actor%20Object%20Reference) (Parameters :type Gameplay%20Cue%20Parameters%20Structure%20%28by%20ref%29))
  (return (|CallParentFunction MyTarget Parameters)))
